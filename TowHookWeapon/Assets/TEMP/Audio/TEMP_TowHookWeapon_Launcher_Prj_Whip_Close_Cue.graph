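# Temporary sound cue for the close-range projectile whip sound of the tow hook weapon's launcher.
# A Random node picks one of three equally weighted whip variations and feeds it through a Modulator for pitch or volume variation.
# The cue uses a dedicated whip attenuation asset, plays in the ranged-weapon projectile effects sound class and carries a slight volume boost.
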SoundCue /TowHookWeapon/Assets/TEMP/Audio/TEMP_TowHookWeapon_Launcher_Prj_Whip_Close_Cue
root=Modulator_0
nodes [
 {id=Modulator_0, children [Random_0]}
 {id=Random_0, children [WavePlayer_0,WavePlayer_1,WavePlayer_2], properties {weights [1,1,1]}}
 {id=WavePlayer_0, properties {soundWaveAssetPtr=&/TEMP_TowHookWeapon_Launcher_Prj_Whip_Close_01}}
 {id=WavePlayer_1, properties {soundWaveAssetPtr=&/TEMP_TowHookWeapon_Launcher_Prj_Whip_Close_02}}
 {id=WavePlayer_2, properties {soundWaveAssetPtr=&/TEMP_TowHookWeapon_Launcher_Prj_Whip_Close_03}}
]
properties {
 attenuationSettings=&/TEMP_TowHookWeapon_Launcher_Prj_Whip_Att
 maxDistance=8500
 soundClassObject=&Fort_WeaponRanged_Projectile_Effects
 volumeMultiplier=1.25
 attenuationOverrides {
  bAttenuate=false
  bEnableReverbSend=false
  bSpatialize=false
  bUseComplexCollisionForOcclusion=true
  falloffDistance=1000
  occlusionVolumeAttenuation=0
  stereoSpread=0
  attenuationShapeExtents {x=512}
 }
}
